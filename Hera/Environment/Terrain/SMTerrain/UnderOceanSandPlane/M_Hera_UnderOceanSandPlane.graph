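# Material for the under-ocean sand plane static mesh in Hera's terrain, compatible with Nanite.
# It is built around one call to the landscape sand layer function MF_Hera_Sand, with a constant full layer weight.
# The function's terrain mask, mud, seaweed and sediment inputs are left unconnected.
# The material uses Material Attributes and Default Lit shading, with tangent-space normals turned off.
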Material /Hera/Environment/Terrain/SMTerrain/UnderOceanSandPlane/M_Hera_UnderOceanSandPlane
graph {
 properties {
  bTangentSpaceNormal=false
  bUseMaterialAttributes=true
  bUsedWithNanite=true
  pixelDepthOffsetMode=PDOM_Legacy
  usedShadingModels=Default Lit
 }
 nodes [MaterialFunctionCall!(
   Material
   MobileMaterial
   LayerWeight=Constant() {r=1}
   UseTangentNormals
   TerrainMasks_QuadA
   TerrainMasks_QuadB
   TerrainMasks_QuadC
   TerrainMasks_QuadD
   VariationMasks Mobile
   OD_MD_Normals
   OD_MD_SedimentLayers
   InputMudLayerMask
   InputSeaweedLayerMask
  ) {materialFunction=/Hera/Environment/Terrain/Landscape/Materials/Layers/MF_Hera_Sand}]
}
